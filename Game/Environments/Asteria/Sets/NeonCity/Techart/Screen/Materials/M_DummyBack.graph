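# M_DummyBack is a minimal Default Lit material in the NeonCity screen tech-art set, apparently a placeholder backing for screens, judging by its name and path.
# Its graph holds only four scalar constants, with no textures, parameters or math, and shows no links to material outputs.
# It is flagged for use with instanced static meshes and Nanite.
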Material /Game/Environments/Asteria/Sets/NeonCity/Techart/Screen/Materials/M_DummyBack
graph {
 properties {bUsedWithInstancedStaticMeshes=true, bUsedWithNanite=true, pixelDepthOffsetMode=PDOM_Legacy, usedShadingModels=Default Lit}
 nodes [
  Constant() {r=0.15}
  Constant() {r=0.666}
  Constant()
  Constant() {r=1}
 ]
}
