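# Material function that hashes a 2D input into a pseudo-random scalar, stored with the NeonCity screen tech-art assets.
# The input, previewed with texture coordinates, is dotted with a constant vector and passed through sine.
# The result is scaled by a large constant and reduced to its fractional part for the Random1D output, following the classic shader sine-hash technique without a noise texture.
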MaterialFunction /Game/Environments/Asteria/Sets/NeonCity/Techart/Screen/MaterialFunctions/MF_Rand2Dto1D
graph {nodes [
  n1 = TextureCoordinate()
  n2 = Sine(None=DotProduct(A=Constant2Vector() {g=4.1414, r=12.9898}, B=FunctionInput(Preview=@n1) {inputType=FunctionInput_Vector2}))
  n3 = Constant() {r=43758.5469}
  FunctionOutput(None=Frac(None=Multiply(A=@n2, B=@n3))) {outputName=Random1D}
 ]}
