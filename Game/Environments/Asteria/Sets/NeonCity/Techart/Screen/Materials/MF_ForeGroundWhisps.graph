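# Material function for NeonCity screen materials that adds optional tinted foreground whisps to an incoming Emissive value.
# Two snowfall-dot samples, distorted by panning normal-map noise, are max-combined and tinted by WhispColor.
# One scrolls the quantized UV input at WhispMovement speed; the other derives coordinates from that movement vector, not mesh UVs.
# A vertical gradient masks the additive result, and the useForeGroundWhisps static switch can bypass it.
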MaterialFunction /Game/Environments/Asteria/Sets/NeonCity/Techart/Screen/Materials/MF_ForeGroundWhisps
graph {nodes [
  Add()
  Constant() {r=0.1}
  n1 = FunctionInput() {inputName=Emissive}
  n2 = AppendVector(
   A=ScalarParameter() {defaultValue=-0.03, parameterName=WhispMovementX}
   B=ScalarParameter() {defaultValue=0.06, parameterName=WhispMovementY}
  )
  n3 = Panner(Coordinate=Multiply(A=@n2) {constB=0.6})
  n4 = TextureCoordinate()
  n5 = TextureSampleParameter2D(UVs=Panner(Coordinate=Multiply(A=@n4) {constB=0.5}) {speedX=0.04, speedY=0.01}) {
   parameterName=Param
   samplerType=SAMPLERTYPE_Normal
   texture=/Game/Characters/Player/Female/Medium/Bodies/F_MED_Bistro_Astronaut/FX/T_GaussianNoise_Normal
  }
  n6 = Multiply(A=ComponentMask(None=@n5.RGB) {g=true, r=true}, B=ScalarParameter() {defaultValue=0.1, parameterName=Offset}) {constB=0.3}
  n7 = TextureSample(UVs=Add(A=Multiply(A=@n3) {constB=1.2}, B=OneMinus(None=@n6))) {texture=/Game/UI/Frontend/ItemShop/Art/Snowfall/Textures/T_UI_SnowfallDots}
  n8 = FunctionInput() {inputName=QUANTIZEDuvS, inputType=FunctionInput_Vector2}
  n9 = TextureSample(UVs=Add(A=Panner(Coordinate=@n8, Speed=@n2), B=@n6)) {texture=&T_UI_SnowfallDots}
  n10 = Add(A=@n1, B=Multiply(A=Max(A=@n7.R, B=@n9.R), B=VectorParameter() {parameterName=WhispColor, defaultValue=<rgba 1 0.0674 0 1>}.RGB))
  n11 = TextureCoordinate()
  n12 = MaterialFunctionCall(UV Channel=@n11) {materialFunction=&LinearGradient}
  n13 = Saturate(None=Subtract(A=@n12.VGradient) {constB=0.4})
  FunctionOutput(None=StaticSwitchParameter(True=LinearInterpolate(A=@n1, B=@n10, Alpha=@n13), False=@n1) {parameterName=useForeGroundWhisps})
  Multiply(A=@n11)
  Multiply(A=@n12.VGradient)
 ]}
