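# Opaque master material for animated LCD advertising screens in Neon City, supporting instanced meshes and Nanite.
# Adaptive-cell UVs pixelate swappable logos over a panning background, with optional gradient and masked pulse effects.
# Noise-driven glitch bars distort the UVs, offset in time by instance position.
# Panel masks, wash color and foreground wisps shape the emissive, which random panel failure, gameplay destruction damage and Fresnel then dim.
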Material /Game/Environments/Asteria/Sets/NeonCity/Techart/Screen/Materials/M_LCD_Ad_Source1
graph {
 properties {
  bUseMaterialAttributes=true
  bUsedWithInstancedStaticMeshes=true
  bUsedWithNanite=true
  pixelDepthOffsetMode=PDOM_Legacy
  usedShadingModels=Default Lit
 }
 nodes [
  n1 = NamedRerouteUsage()
  n2 = NamedRerouteUsage()
  n3 = NamedRerouteUsage()
  n4 = TextureSampleParameter2D!(UVs=@n2, Bias=@n3, Apply View MipBias) {
   mipValueMode=TMVM_MipBias
   parameterName=Primary Logo
   texture=/Game/Environments/Asteria/Sets/NeonCity/Techart/Screen/Textures/Logos/TeddyBear
  }
  n5 = TextureSampleParameter2D(UVs=@n2, Bias=@n3) {
   mipValueMode=TMVM_MipBias
   parameterName=SwapLogo1
   samplerSource=SSM_Clamp_WorldGroupSettings
   texture=&TeddyBear
  }
  n6 = Multiply(A=Time(), B=ScalarParameter() {defaultValue=0.02, parameterName=SignSwapSpeed})
  n7 = Sine(None=@n6)
  n8 = Power(Base=OneMinus(None=Abs(None=@n7))) {constExponent=12}
  n9 = ConstantBiasScale(None=Sine(None=@n6)) {scale=1}
  n10 = Saturate(None=StaticSwitchParameter(True=Round(None=@n8), False=Power(Base=@n9) {constExponent=12}) {parameterName=blink})
  n11 = LinearInterpolate(A=@n4.RGB, B=@n5.RGB, Alpha=@n10)
  n12 = LinearInterpolate(A=@n4.A, B=@n5.A, Alpha=@n10)
  n13 = TextureSampleParameter2D!(UVs=NamedRerouteDeclaration() {name=ScaledUVs_1, nodeColor=<rgba 1 0 0.0938 1>}, Apply View MipBias) {parameterName=Gradient, texture=/Game/Global/Textures/Default/Masks_and_Gradients/CylGradient}
  n14 = ScalarParameter() {defaultValue=0.1, parameterName=BG_HorizontalPanningSpeed}
  n15 = ScalarParameter() {defaultValue=0.2, parameterName=BG_VerticalPanningSpeed}
  n16 = TextureSampleParameter2D!(
   UVs=Panner(
    Coordinate=NamedRerouteUsage()
    Speed=MaterialFunctionCall(X=@n14, Y=@n15) {materialFunction=&MakeFloat2}.Result
   )
   Bias=NamedRerouteUsage()
   Apply View MipBias
  ) {
   mipValueMode=TMVM_MipBias
   parameterName=BackGroundTexture
   samplerType=SAMPLERTYPE_Masks
   texture=/Game/Environments/Asteria/Sets/NeonCity/Techart/Screen/Textures/panelRGB
  }
  n17 = StaticSwitchParameter(
   True=@n16.RGB
   False=LinearInterpolate(
    A=VectorParameter() {parameterName=ColorBG, defaultValue=<rgba 0 0.4913 1 1>}.RGB
    B=VectorParameter() {parameterName=ColorAccent, defaultValue=<rgba 0.6189 0.803 0.974 1>}.RGB
    Alpha=ChannelMaskParameter(None=@n16.RGB) {parameterName=BG Mask Channel}.R
   )
  ) {parameterName=FullColorBackground}
  n18 = StaticSwitchParameter(True=Multiply(A=ConstantBiasScale(None=@n13.RGB) {bias=0.5, scale=1}, B=@n17), False=@n17) {parameterName=BG Gradient}
  n19 = StaticSwitchParameter(
   True=TextureSample(UVs=NamedRerouteUsage()) {texture=/Game/Weapons/WeaponSkins/Wraps/Textures/UVCheaker}.RGB
   False=StaticSwitchParameter(True=LinearInterpolate(A=@n18, B=@n11, Alpha=@n12), False=LinearInterpolate(A=@n18, B=@n4.RGB, Alpha=@n4.RGB)) {parameterName=RotatingLogos}
  ) {group=Debug, parameterName=DebugDirectionTexture}
  n20 = ScalarParameter() {defaultValue=4, parameterName=Panner Boost Red}
  n21 = NamedRerouteUsage()
  n22 = TextureSampleParameter2D(UVs=@n2, Bias=@n3) {
   mipValueMode=TMVM_MipBias
   parameterName=logo fx mask
   samplerType=SAMPLERTYPE_Masks
   texture=/Game/Environments/Asteria/Sets/NeonCity/Techart/Screen/Textures/Logos/Art_SwordGirl_mASKED
  }
  n23 = Multiply(
   A=TextureSample!(UVs=Panner(Coordinate=@n21) {speedY=0.5}, Tex, Bias=Constant() {r=3}, Apply View MipBias) {mipValueMode=TMVM_MipBias, texture=&CylGradient}.R
   B=@n22.R
  )
  n24 = Add(A=@n19, B=Multiply(A=@n19, B=Multiply(A=@n20, B=@n23)))
  n25 = Frac(None=Multiply(A=Time(), B=ScalarParameter() {defaultValue=0.5, parameterName=Pulse Speed}))
  n26 = Multiply(A=@n24, B=LinearInterpolate(Alpha=Round(None=@n25)))
  n27 = Clamp(None=StaticSwitchParameter(True=LinearInterpolate(A=@n24, B=@n26, Alpha=@n22.B), False=@n19) {parameterName=Logo SFX}) {clampMode=CMODE_ClampMax, maxDefault=0.99}
  n28 = NamedRerouteUsage()
  n29 = Subtract(A=Power(Base=TextureSample!(UVs=@n28, Tex, Apply View MipBias) {samplerType=SAMPLERTYPE_Masks, texture=/Game/Environments/Asteria/Sets/NeonCity/Techart/Textures/T_LCD_Panels_output}.R) {constExponent=50}) {constB=-0.5}
  n30 = ScalarParameter() {defaultValue=0.02, parameterName=RandomPanelStrength}
  n31 = Add(A=@n27, B=Multiply(A=@n27, B=Multiply(A=@n29, B=@n30)))
  n32 = Saturate(None=VectorParameter() {parameterName=WashColor, defaultValue=<rgba 0 27.3061 100 0>}.RGB)
  n33 = ScalarParameter() {defaultValue=0.666, parameterName=EmissiveWashFactor}
  n34 = MaterialFunctionCall!(
   QUANTIZEDuvS=NamedRerouteUsage()
   Emissive=Multiply(
    A=TextureSample(UVs=@n1) {samplerType=SAMPLERTYPE_Masks, texture=/Game/Environments/Asteria/Sets/NeonCity/Techart/Textures/T_LCD_Panels_LCDSoft}.RGB
    B=LinearInterpolate(A=@n31, B=@n32, Alpha=@n33)
   )
  ) {materialFunction=&/MF_ForeGroundWhisps}
  Desaturation(
   None=@n34.Result
   Fraction=MaterialFunctionCall(
    Value_Morning=ScalarParameter() {defaultValue=0.1, parameterName=Desaturation_Morning}
    Value_Day=ScalarParameter() {defaultValue=0.3, parameterName=Desaturation_Day}
    Value_Evening=ScalarParameter() {defaultValue=0.1, parameterName=Desaturation_Evening}
    Value_Night=Constant()
   ) {materialFunction=&Scalar_Time_of_Day}.Result
  )
  MakeMaterialAttributes!(
   BaseColor=NamedRerouteUsage()
   Metallic=NamedRerouteUsage()
   Specular
   Roughness=NamedRerouteUsage()
   Anisotropy
   EmissiveColor=NamedRerouteUsage()
   Opacity
   OpacityMask
   Normal=NamedRerouteUsage()
   Tangent
   WorldPositionOffset=MaterialFunctionCall() {materialFunction=&MF_FN_DestructibleDamage}.DamageAnimWPO
   SubsurfaceColor
   ClearCoat
   ClearCoatRoughness
   AmbientOcclusion
   Refraction
   CustomizedUVs_0
   CustomizedUVs_1
   CustomizedUVs_2
   CustomizedUVs_3
   CustomizedUVs_4
   CustomizedUVs_5
   CustomizedUVs_6
   CustomizedUVs_7
   PixelDepthOffset
   ShadingModel
   Displacement
  )
  MaterialFunctionCall(Value=@n6) {materialFunction=&LinearSine}
  n35 = ScalarParameter() {parameterName=BG_EdgeOffset}
  n36 = ScalarParameter() {defaultValue=1, parameterName=Primary Logo Scale X}
  n37 = ScalarParameter() {defaultValue=1, parameterName=Primary Logo Scale Y}
  n38 = Max(A=PerInstanceCustomData!(DefaultValue), B=ScalarParameter() {bUseCustomPrimitiveData=true, parameterName=CellSize})
  n39 = MaterialFunctionCall!(
   bump
   offsetToEdge=@n35
   uvscale=StaticSwitchParameter(True=AppendVector(A=@n36, B=@n37), False=@n36) {parameterName=NonSquaredLogo}
   PixelAspect
   cellSize=@n38
  ) {materialFunction=/Game/Environments/Asteria/Sets/NeonCity/Techart/Screen/MaterialFunctions/MF_AdaptiveCellsUVs}
  n40 = MaterialFunctionCall!(
   PrimaryTextureCoordinates=ComponentMask(None=NamedRerouteDeclaration(None=@n39.tilingUVs) {name=ScaledUVs, nodeColor=<rgba 1 0 0.0938 1>}) {r=true}
   2ndOrderTextureCoordinates
   RandomSeed=Constant() {r=8}
   NoiseDensityBias
   Bypass Noise Influence Crosstalk
  ) {materialFunction=/Game/Effects/Fort_Effects/Materials/_Functions/MF_RandomUVSample}
  n41 = Constant3Vector()
  n42 = Fmod(
   A=DotProduct(
    A=TransformPosition(None=@n41) {transformSourceType=TRANSFORMPOSSOURCE_Instance, transformType=TRANSFORMPOSSOURCE_World}
    B=Constant3Vector() {constant=<rgba 5.4623 1.2346 9.2232 0>}
   )
   B=Constant() {r=50814.0312}
  )
  n43 = MaterialFunctionCall!(Frequency Coefficients, Wave Weights, In=Add(A=Time(), B=@n42)) {materialFunction=/Game/Environments/Asteria/Sets/NeonCity/Techart/Screen/MaterialFunctions/MF_Cheap1DNoise}
  n44 = Constant() {r=0.8}
  n45 = Saturate(None=Sign(None=Subtract(A=@n43.Noise_0_1, B=@n44)))
  n46 = ComponentMask(None=NamedRerouteUsage()) {r=true}
  n47 = Add(A=Time(), B=Add(A=NamedRerouteUsage()) {constB=13.2446})
  n48 = ScalarParameter() {defaultValue=0.04, parameterName=randSpeed}
  n49 = Multiply(A=MaterialFunctionCall(In=Multiply(A=@n47, B=@n48)) {materialFunction=&MF_Cheap1DNoise}.Noise_0_1) {constB=128}
  n50 = Saturate(None=MaterialFunctionCall!(In=Floor(None=@n49)) {materialFunction=/Game/Environments/Asteria/Sets/NeonCity/Techart/Screen/MaterialFunctions/MF_Rand1Dto1D1}.Random1D)
  n51 = Sign(None=Subtract(A=@n46, B=@n50))
  n52 = Add(A=@n46, B=ScalarParameter() {defaultValue=0.2, parameterName=GlitchBarThickness}) {constB=0.5}
  n53 = Sign(None=Multiply(A=Subtract(A=@n52, B=@n50)) {constB=-1})
  n54 = OneMinus(None=Saturate(None=Max(A=@n51, B=@n53)))
  n55 = Multiply(
   A=AppendVector(A=Constant(), B=Subtract(A=@n40.NoiseWithRange) {constB=0.5})
   B=Multiply(A=ScalarParameter() {defaultValue=0.02, parameterName=DistortionScale}, B=Multiply(A=@n45, B=@n54))
  )
  n56 = Add(A=@n39.clamped_uvs, B=@n55)
  n57 = MaterialFunctionCall(Float2=@n56) {materialFunction=&BreakOutFloat2Components}
  n58 = OneMinus(None=@n57.R)
  NamedRerouteDeclaration(None=StaticSwitchParameter(
    True=MaterialFunctionCall(X=@n58, Y=@n57.G) {materialFunction=&MakeFloat2}.Result
    False=@n56
   ) {group=Debug, parameterName=FlipUVsHorizontally}) {name=QuantizedUVs, nodeColor=<rgba 1 0.8672 0 1>}
  NamedRerouteDeclaration(None=NamedRerouteUsage()) {name=normal, nodeColor=<rgba 0.2891 0 1 1>}
  NamedRerouteDeclaration(None=If(A=PerInstanceCustomData(), "A > B"=ScalarParameter() {defaultValue=2, parameterName=fdsa}, "A < B"=Constant()) {constB=0.001}) {name=HLODBiasValue, nodeColor=<rgba 1 0.1875 0 1>}
  n59 = ScalarParameter() {defaultValue=1, parameterName="Background Tiling "}
  n60 = AppendVector(A=@n59, B=Multiply(A=@n59, B=ScalarParameter() {defaultValue=1, parameterName=BGStretch}))
  n61 = ScalarParameter() {defaultValue=63.3362, parameterName=BlankPanelSizeInPixels}
  NamedRerouteDeclaration(None=MaterialFunctionCall(uvscale=@n60, PixelAspect=Constant() {r=1.7778}, cellSize=Multiply(A=@n38, B=@n61) {constB=9}) {materialFunction=&MF_AdaptiveCellsUVs}.unclamped_uvs) {name=unclamped ratio, nodeColor=<rgba 0.7578 0 1 1>}
  NamedRerouteDeclaration(None=@n42) {name=perACtorRand, nodeColor=<rgba 0 0.9531 1 1>}
  NamedRerouteDeclaration(None=Add(
    A=MaterialFunctionCall(offsetToEdge=@n35, uvscale=@n60, cellSize=@n38) {materialFunction=&MF_AdaptiveCellsUVs}.unclamped_uvs
    B=@n55
   )) {name=BG_unclamped_pixelated, nodeColor=<rgba 0 1 0.4375 1>}
  n62 = TextureSample(UVs=TextureCoordinate() {uTiling=2, vTiling=2}) {samplerType=SAMPLERTYPE_Masks, texture=/Game/Environments/Asteria/Sets/NeonCity/Techart/Textures/T_GlassA_Masks}
  NamedRerouteDeclaration(None=LinearInterpolate(
    A=ScalarParameter() {defaultValue=0.1, parameterName=MetallicLow}
    B=ScalarParameter() {defaultValue=0.12, parameterName=MetallicHigh}
    Alpha=@n62.R
   )) {name=Metallic, nodeColor=<rgba 1 0.2812 0 1>}
  NamedRerouteDeclaration(None=LinearInterpolate(
    A=ScalarParameter() {defaultValue=0.1, parameterName=RoughnessLow}
    B=ScalarParameter() {defaultValue=0.15, parameterName=RoughnessHigh}
    Alpha=@n62.R
   )) {name=Roughness, nodeColor=<rgba 1 0.0234 0 1>}
  n63 = TextureSample() {samplerType=SAMPLERTYPE_Normal, texture=/Game/Characters/Player/Female/Medium/Bodies/F_MED_Bistro_Astronaut/FX/T_GaussianNoise_Normal}
  NamedRerouteDeclaration(None=Multiply(A=@n63.RGB, B=Constant3Vector() {constant=<rgba 0.015 0.015 1 0>})) {name=Normals, nodeColor=<rgba 1 0 0.6797 1>}
  NamedRerouteDeclaration(None=@n39.UnPixelatedUVNoClamp) {name=unclampedunpixel, nodeColor=<rgba 1 0.8438 0 1>}
  n64 = MaterialFunctionCall!(Emissive=@n34.Result) {materialFunction=/Game/Environments/Asteria/Sets/NeonCity/Techart/Screen/MaterialFunctions/MF_NeonCityEmissiveBalance}
  n65 = NamedRerouteUsage()
  n66 = MaterialFunctionCall!(In=@n65) {materialFunction=/Game/Environments/Asteria/Sets/NeonCity/Techart/Screen/MaterialFunctions/MF_Rand2Dto1D}
  n67 = NamedRerouteUsage()
  n68 = Add(A=Multiply(A=Time()) {constB=0.15}, B=Multiply(A=@n66.Random1D, B=Add(A=@n66.Random1D, B=@n67)) {constB=17.3333})
  n69 = Sign(None=Subtract(A=MaterialFunctionCall(In=@n68) {materialFunction=&MF_Cheap1DNoise}.Noise_0_1) {constB=0.4})
  n70 = MaterialFunctionCall(In=Add(A=@n65, B=Constant2Vector() {g=53, r=17}) {constB=172}) {materialFunction=&MF_Rand2Dto1D}
  n71 = ScalarParameter() {defaultValue=0.1, parameterName=PanelFailChance}
  n72 = Saturate(None=Sign(None=Subtract(A=@n70.Random1D, B=@n71)))
  n73 = OneMinus(None=Multiply(A=Saturate(None=@n69), B=OneMinus(None=@n72)))
  n74 = ScalarParameter() {
   bUseCustomPrimitiveData=true
   desc="Destructible Damage. This parameter is controlled by gameplay code.\r\nDefault=Custom Primitive Data Index 13"
   group=z_GameplayParameters_DestructibleDamage
   parameterName=Code_DestructionAmount
   primitiveDataIndex=13
   sortPriority=105
  }
  n75 = MaterialFunctionCall(Input=@n74, Target Low=Constant() {r=0.8}, Target High=Constant() {r=0.4}) {materialFunction=&RemapValueRange}
  n76 = Multiply(A=Multiply(A=@n74, B=@n70.Random1D)) {constB=0.6}
  n77 = Saturate(None=Multiply(
    A=MaterialFunctionCall(Input=@n73, Target Low=@n75.Result) {materialFunction=&RemapValueRange}.Result
    B=OneMinus(None=@n76)
   ))
  n78 = Fresnel() {baseReflectFraction=0, exponent=2}
  NamedRerouteDeclaration(None=Multiply(A=Multiply(A=@n64.Result, B=@n77), B=OneMinus(None=@n78))) {name=emissive, nodeColor=<rgba 1 0.75 0 1>}
  NamedRerouteDeclaration(None=Add(A=VectorParameter() {parameterName=BaseColor, defaultValue=<rgba 0.1115 0.125 0.0981 1>}.RGB, B=@n34.Result)) {name=Base Color, nodeColor=<rgba 1 0 0.3516 1>}
 ]
}
