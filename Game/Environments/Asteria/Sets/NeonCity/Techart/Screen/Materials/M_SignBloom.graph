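# Unlit additive material for a fake bloom glow on NeonCity signs, with GlowColor exposed through custom primitive data.
# A UV sphere mask is multiplied by a Fresnel term, pixel-depth remap, time-of-day alpha and pivot-to-camera distance fade, then passed through a VertexInterpolator.
# The glow fades out as the camera nears.
# When time-of-day alpha or distance fade nears zero, a vertex offset collapses the mesh to the actor position.
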Material /Game/Environments/Asteria/Sets/NeonCity/Techart/Screen/Materials/M_SignBloom
graph {
 properties {
  blendMode=BLEND_Additive
  pixelDepthOffsetMode=PDOM_Legacy
  previewMesh=/Game/Environments/Asteria/Props/NeonCity/NeonCity_Signage/Blueprints/SM_NeonCity_Switch_FakeGlow
  shadingModel=MSM_Unlit
  usedShadingModels=Unlit
 }
 nodes [
  n1 = CollectionParameter() {
   collection=&FortniteMaterialParameters
   parameterName=FortniteActiveTimeOfDay
  }
  n2 = Constant4Vector() {constant=<rgba 0 0 0 1>}
  n3 = MaterialFunctionCall() {materialFunction=&ObjectPivotPoint}
  n4 = CameraPositionWS()
  n5 = Saturate(None=InverseLinearInterpolate(
    A=ScalarParameter() {defaultValue=500, parameterName=EndFade}
    B=ScalarParameter() {defaultValue=1000, parameterName=StartFade}
    Value=Distance(A=@n3.Object Pivot Location, B=@n4)
   ))
  If(A=Min(A=DotProduct(A=@n1, B=@n2.RGBA), B=@n5), "A > B"=Constant(), "A < B"=Subtract(A=ActorPositionWS(), B=WorldPosition().XYZ)) {constB=0.01}
  VectorParameter() {bUseCustomPrimitiveData=true, parameterName=GlowColor, defaultValue=<rgba 1 1 1 1>}
  n6 = TextureCoordinate()
  n7 = Power(Base=Abs(None=ConstantBiasScale(None=@n6) {bias=-0.5, scale=2}))
  n8 = Constant()
  n9 = Constant() {r=1}
  n10 = Constant() {r=0.6}
  n11 = MaterialFunctionCall(Input=PixelDepth().R, Input Low=Constant() {r=500}, Input High=Constant() {r=6500}) {materialFunction=&RemapValueRange}
  n12 = ScalarParameter() {defaultValue=3, parameterName=FesnelBrightness}
  n13 = ScalarParameter() {defaultValue=2, parameterName=FresnelExponent}
  n14 = Normalize(VectorInput=Subtract(A=CameraPositionWS(), B=WorldPosition().XYZ))
  n15 = CollectionParameter() {
   collection=&FortniteMaterialParameters
   parameterName=FortniteActiveTimeOfDay
  }
  n16 = Multiply(
   A=Multiply(
    A=Saturate(None=@n11.Result)
    B=MaterialFunctionCall(Brightness=@n12, Exponent=@n13, Camera Vector=@n14) {materialFunction=&MF_Fresnel_FX}.Fresnel
   )
   B=ComponentMask(None=@n15) {a=true}
  )
  VertexInterpolator(VS=Multiply(A=SphereMask(A=@n7, B=@n8, Radius=@n9, Hardness=@n10) {hardnessPercent=0.5}, B=Multiply(A=@n16, B=@n5)))
 ]
}
